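# Sound cue for the Tropical pickaxe cosmetic's impact sound when it strikes.
# A short fixed delay feeds an evenly weighted random pick among four impact wave variations.
# It uses the shared pickaxe impact attenuation, impacts sound class, cosmetic property sheet and player melee weapon concurrency set, and primes on load.
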SoundCue /BRCosmetics/Sounds/PickAxes/Tropical/PA_Tropical_Impact_Cue
root=Delay_1
nodes [
 {id=Delay_1, children [Random_1], properties {delayMax=0.1, delayMin=0.1}}
 {id=Random_1, children [WavePlayer_3,WavePlayer_5,WavePlayer_6,WavePlayer_7], properties {weights [1,1,1,1]}}
 {id=WavePlayer_3, properties {soundWaveAssetPtr=&/Pickaxe_Tropical_Impact_01}}
 {id=WavePlayer_5, properties {soundWaveAssetPtr=&/Pickaxe_Tropical_Impact_02}}
 {id=WavePlayer_6, properties {soundWaveAssetPtr=&/Pickaxe_Tropical_Impact_03}}
 {id=WavePlayer_7, properties {soundWaveAssetPtr=&/Pickaxe_Tropical_Impact_04}}
]
properties {
 attenuationSettings=/Game/Sounds/Attenuation/Fort_Pickaxe_Impact_Att
 audioPropertiesSheet=/BRCosmetics/Sounds/Settings/PropertySheets/Pickaxe/APS_Cosmetic_Pickaxe_Impact
 bPrimeOnLoad=true
 maxDistance=4300
 soundClassObject=&Fort_Impacts
 volumeMultiplier=0.9
 attenuationOverrides {
  bAttenuateWithLPF=true
  distanceAlgorithm=Logarithmic
  falloffDistance=800
  lPFRadiusMax=800
  lPFRadiusMin=400
  attenuationShapeExtents {x=300}
 }
 concurrencyOverrides {bLimitToOwner=true, maxCount=2, resolutionRule=StopOldest}
 concurrencySet [&Player_Melee_Weapon_Conc]
}
